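# Sound cue for the jump foley of metal-pawed quadruped cosmetic companions.
# An equal-weight Random node picks one of four jump waves, feeding a Modulator that slightly varies pitch and volume.
# It plays quietly in the pets sound class with companion foley attenuation and zero priority, and shares the biped footstep concurrency set.
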SoundCue /CosmeticCompanions/Assets/Audio_Shared/Foley/Quadraped/Metal_Paws/SC_MImosa_Quadruped_MetalPaws_Jump
root=Modulator_0
nodes [
 {id=Modulator_0, children [Random_2], properties {pitchMax=1.1, pitchMin=0.9, volumeMax=1, volumeMin=0.9}}
 {id=Random_2, children [WavePlayer_0,WavePlayer_1,WavePlayer_2,WavePlayer_3], properties {weights [1,1,1,1]}}
 {
  id=WavePlayer_0
  properties {soundWaveAssetPtr=/CosmeticCompanions/Assets/Audio_Shared/Foley/Quadraped/Metal_Paws/Wavs/Mimosa_Quadruped_MetalPaws_Jump_01}
 }
 {
  id=WavePlayer_1
  properties {soundWaveAssetPtr=/CosmeticCompanions/Assets/Audio_Shared/Foley/Quadraped/Metal_Paws/Wavs/Mimosa_Quadruped_MetalPaws_Jump_02}
 }
 {
  id=WavePlayer_2
  properties {soundWaveAssetPtr=/CosmeticCompanions/Assets/Audio_Shared/Foley/Quadraped/Metal_Paws/Wavs/Mimosa_Quadruped_MetalPaws_Jump_03}
 }
 {
  id=WavePlayer_3
  properties {soundWaveAssetPtr=/CosmeticCompanions/Assets/Audio_Shared/Foley/Quadraped/Metal_Paws/Wavs/Mimosa_Quadruped_MetalPaws_Jump_04}
 }
]
properties {
 attenuationSettings=/Game/Sounds/Attenuation/AS_Companion_Foley
 maxDistance=1200
 priority=0
 soundClassObject=&Fort_Creatures_Pets
 volumeMultiplier=0.25
 concurrencySet [/CosmeticCompanions/Assets/Audio_Shared/Foley/Biped/CONC_Mimosa_Biped_Footstep]
}
