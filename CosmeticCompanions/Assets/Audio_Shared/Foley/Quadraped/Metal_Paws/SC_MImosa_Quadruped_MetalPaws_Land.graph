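# Sound cue for metal-paw landing foley on quadruped companion pets.
# A quality-level node sends the first two levels to a random pick of five landing waves and the last two to a pick of the first four.
# A modulator adds slight pitch and volume variation to the result.
# It plays quietly with companion foley attenuation, the pets sound class and a concurrency setting shared from biped footsteps.
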SoundCue /CosmeticCompanions/Assets/Audio_Shared/Foley/Quadraped/Metal_Paws/SC_MImosa_Quadruped_MetalPaws_Land
root=Modulator_0
nodes [
 {id=Modulator_0, children [QualityLevel_0], properties {pitchMax=1.1, pitchMin=0.9, volumeMax=1, volumeMin=0.9}}
 {id=QualityLevel_0, children [Random_0,Random_0,Random_3,Random_3]}
 {id=Random_0, children [WavePlayer_11, WavePlayer_12, WavePlayer_13, WavePlayer_14, WavePlayer_15], properties {weights [1,1,1,1,1]}}
 {id=Random_3, children [WavePlayer_16, WavePlayer_19, WavePlayer_20, WavePlayer_21], properties {weights [1,1,1,1]}}
 {
  id=WavePlayer_11
  properties {soundWaveAssetPtr=/CosmeticCompanions/Assets/Audio_Shared/Foley/Quadraped/Metal_Paws/Wavs/Mimosa_Quadruped_MetalPaws_Land_01}
 }
 {
  id=WavePlayer_12
  properties {soundWaveAssetPtr=/CosmeticCompanions/Assets/Audio_Shared/Foley/Quadraped/Metal_Paws/Wavs/Mimosa_Quadruped_MetalPaws_Land_02}
 }
 {
  id=WavePlayer_13
  properties {soundWaveAssetPtr=/CosmeticCompanions/Assets/Audio_Shared/Foley/Quadraped/Metal_Paws/Wavs/Mimosa_Quadruped_MetalPaws_Land_03}
 }
 {
  id=WavePlayer_14
  properties {soundWaveAssetPtr=/CosmeticCompanions/Assets/Audio_Shared/Foley/Quadraped/Metal_Paws/Wavs/Mimosa_Quadruped_MetalPaws_Land_04}
 }
 {
  id=WavePlayer_15
  properties {soundWaveAssetPtr=/CosmeticCompanions/Assets/Audio_Shared/Foley/Quadraped/Metal_Paws/Wavs/Mimosa_Quadruped_MetalPaws_Land_05}
 }
 {
  id=WavePlayer_16
  properties {soundWaveAssetPtr=&Mimosa_Quadruped_MetalPaws_Land_01}
 }
 {
  id=WavePlayer_19
  properties {soundWaveAssetPtr=&Mimosa_Quadruped_MetalPaws_Land_02}
 }
 {
  id=WavePlayer_20
  properties {soundWaveAssetPtr=&Mimosa_Quadruped_MetalPaws_Land_03}
 }
 {
  id=WavePlayer_21
  properties {soundWaveAssetPtr=&Mimosa_Quadruped_MetalPaws_Land_04}
 }
]
properties {
 attenuationSettings=/Game/Sounds/Attenuation/AS_Companion_Foley
 maxDistance=1200
 priority=0
 soundClassObject=&Fort_Creatures_Pets
 volumeMultiplier=0.16
 concurrencySet [/CosmeticCompanions/Assets/Audio_Shared/Foley/Biped/CONC_Mimosa_Biped_Footstep]
}
