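# Sound cue for the sit foley of a metal-pawed quadruped cosmetic companion (Mimosa).
# A Random node picks one of three sit wave variations with equal weight.
# Playback uses the companion foley attenuation and pets sound class at lowered volume and zero priority, and shares the quadruped footstep concurrency group.
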SoundCue /CosmeticCompanions/Assets/Audio_Shared/Foley/Quadraped/Metal_Paws/SC_MImosa_Quadruped_MetalPaws_Sit
root=Random_1
nodes [
 {id=Random_1, children [WavePlayer_6,WavePlayer_7,WavePlayer_8], properties {weights [1,1,1]}}
 {
  id=WavePlayer_6
  properties {soundWaveAssetPtr=/CosmeticCompanions/Assets/Audio_Shared/Foley/Quadraped/Metal_Paws/Wavs/Mimosa_Quadruped_MetalPaws_Sit_03}
 }
 {
  id=WavePlayer_7
  properties {soundWaveAssetPtr=/CosmeticCompanions/Assets/Audio_Shared/Foley/Quadraped/Metal_Paws/Wavs/Mimosa_Quadruped_MetalPaws_Sit_02}
 }
 {
  id=WavePlayer_8
  properties {soundWaveAssetPtr=/CosmeticCompanions/Assets/Audio_Shared/Foley/Quadraped/Metal_Paws/Wavs/Mimosa_Quadruped_MetalPaws_Sit_01}
 }
]
properties {
 attenuationSettings=/Game/Sounds/Attenuation/AS_Companion_Foley
 maxDistance=1200
 priority=0
 soundClassObject=&Fort_Creatures_Pets
 volumeMultiplier=0.45
 concurrencySet [/CosmeticCompanions/Assets/Audio_Shared/Foley/Quadraped/CONC_Mimosa_Quadraped_Footstep]
}
