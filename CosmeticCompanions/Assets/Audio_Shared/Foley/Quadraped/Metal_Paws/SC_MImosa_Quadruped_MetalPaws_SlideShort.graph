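# Sound cue for short slide foley on metal-pawed quadruped cosmetic companions, in the creature pets sound class.
# A quality-level root sends its first two levels to an equal-weight random pick of five slide waves and its last two to a pick of the first four.
# Playback runs at half volume with companion foley attenuation and borrows the biped footstep concurrency set.
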SoundCue /CosmeticCompanions/Assets/Audio_Shared/Foley/Quadraped/Metal_Paws/SC_MImosa_Quadruped_MetalPaws_SlideShort
root=QualityLevel_1
nodes [
 {id=QualityLevel_1, children [Random_2,Random_2,Random_3,Random_3]}
 {id=Random_2, children [WavePlayer_11, WavePlayer_12, WavePlayer_13, WavePlayer_14, WavePlayer_15], properties {weights [1,1,1,1,1]}}
 {id=Random_3, children [WavePlayer_16, WavePlayer_19, WavePlayer_20, WavePlayer_21], properties {weights [1,1,1,1]}}
 {
  id=WavePlayer_11
  properties {soundWaveAssetPtr=/CosmeticCompanions/Assets/Audio_Shared/Foley/Quadraped/Metal_Paws/Wavs/Mimosa_Quadruped_MetalPaws_SlideShort_01}
 }
 {
  id=WavePlayer_12
  properties {soundWaveAssetPtr=/CosmeticCompanions/Assets/Audio_Shared/Foley/Quadraped/Metal_Paws/Wavs/Mimosa_Quadruped_MetalPaws_SlideShort_02}
 }
 {
  id=WavePlayer_13
  properties {soundWaveAssetPtr=/CosmeticCompanions/Assets/Audio_Shared/Foley/Quadraped/Metal_Paws/Wavs/Mimosa_Quadruped_MetalPaws_SlideShort_03}
 }
 {
  id=WavePlayer_14
  properties {soundWaveAssetPtr=/CosmeticCompanions/Assets/Audio_Shared/Foley/Quadraped/Metal_Paws/Wavs/Mimosa_Quadruped_MetalPaws_SlideShort_04}
 }
 {
  id=WavePlayer_15
  properties {soundWaveAssetPtr=/CosmeticCompanions/Assets/Audio_Shared/Foley/Quadraped/Metal_Paws/Wavs/Mimosa_Quadruped_MetalPaws_SlideShort_05}
 }
 {
  id=WavePlayer_16
  properties {soundWaveAssetPtr=&Mimosa_Quadruped_MetalPaws_SlideShort_01}
 }
 {
  id=WavePlayer_19
  properties {soundWaveAssetPtr=&Mimosa_Quadruped_MetalPaws_SlideShort_02}
 }
 {
  id=WavePlayer_20
  properties {soundWaveAssetPtr=&Mimosa_Quadruped_MetalPaws_SlideShort_03}
 }
 {
  id=WavePlayer_21
  properties {soundWaveAssetPtr=&Mimosa_Quadruped_MetalPaws_SlideShort_04}
 }
]
properties {
 attenuationSettings=/Game/Sounds/Attenuation/AS_Companion_Foley
 maxDistance=1200
 priority=0
 soundClassObject=&Fort_Creatures_Pets
 volumeMultiplier=0.5
 concurrencySet [/CosmeticCompanions/Assets/Audio_Shared/Foley/Biped/CONC_Mimosa_Biped_Footstep]
}
